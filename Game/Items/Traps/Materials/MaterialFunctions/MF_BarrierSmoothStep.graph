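# MF_BarrierSmoothStep is a library-exposed material function under trap materials that turns a scalar into a smooth ramp.
# It offsets and halves the input, feeds it to a Sine, then bias-scales and saturates the wave, easing 0 to 1 along a cosine curve.
# The input is never clamped, so values outside 0 to 1 oscillate instead of holding at the ends like a true smoothstep.
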MaterialFunction /Game/Items/Traps/Materials/MaterialFunctions/MF_BarrierSmoothStep
graph {
 properties {bExposeToLibrary=true, description=Smooth ramp, libraryCategoriesText []}
 nodes [
  n1 = FunctionInput() {inputName=Input, inputType=FunctionInput_Scalar}
  n2 = Sine(None=Multiply(A=Add(A=@n1) {constB=1.5}) {constB=0.5})
  FunctionOutput(None=Saturate(None=ConstantBiasScale(None=@n2))) {outputName=Output}
 ]
}
